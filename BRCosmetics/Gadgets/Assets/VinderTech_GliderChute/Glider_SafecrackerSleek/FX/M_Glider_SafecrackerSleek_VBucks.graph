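# Lit, two-sided material for V-Bucks mesh particles in the Safecracker Sleek glider's Niagara effects.
# The V-Buck texture's inverted red channel blends outer and inner colors and scalar surface values, while boosted, saturated green masks the inner color.
# Vertex offset adds a Jiggle_WPO wave, driven by particle age and dynamic speed, to normal-direction displacement weighted by vertex red/green and dynamic parameters.
# The "Use on Pile" static switch zeroes that motion.
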Material /BRCosmetics/Gadgets/Assets/VinderTech_GliderChute/Glider_SafecrackerSleek/FX/M_Glider_SafecrackerSleek_VBucks
graph {
 properties {
  bCastRayTracedShadows=false
  bUsedWithNiagaraMeshParticles=true
  pixelDepthOffsetMode=PDOM_Legacy
  translucencyLightingMode=TLM_VolumetricPerVertexDirectional
  twoSided=true
  usedShadingModels=Default Lit
 }
 nodes [
  n1 = TextureSample() {texture=&/T_FX_Glider_SafecrackerSleek_Vbuck}
  n2 = OneMinus(None=@n1.R)
  LinearInterpolate(Alpha=@n2) {constA=0.2, constB=0.8}
  LinearInterpolate(Alpha=@n2) {constA=0.4, constB=0}
  LinearInterpolate(Alpha=@n2) {constA=1, constB=0.05}
  n3 = LinearInterpolate(A=Constant(), B=Constant() {r=25}, Alpha=@n1.G)
  LinearInterpolate(
   A=VectorParameter() {parameterName=OuterColor, defaultValue=<rgba 1 0.5359 0.1562 1>}.RGB
   B=Multiply(A=VectorParameter() {parameterName=InnerColor, defaultValue=<rgba 0.065 0.065 0.065 1>}.RGB, B=Saturate(None=@n3))
   Alpha=@n2
  )
  n4 = DynamicParameter() {defaultValue=<rgba 0 2 0 1>, paramNames [ColorChoice,WPO_Speed,WPO_Red,WPO_Green]}
  n5 = ScalarParameter() {defaultValue=0.5, parameterName=Small Wave}
  n6 = Multiply(A=ParticleRelativeTime(), B=ScalarParameter() {defaultValue=10, parameterName=WaveScale})
  n7 = VertexNormalWS()
  n8 = ScalarParameter() {defaultValue=2, parameterName=Amount_Offset_Red}
  n9 = Multiply(A=@n8) {constB=-1}
  n10 = DynamicParameter() {defaultValue=<rgba 0 2 0 1>, paramNames [ColorChoice,WPO_Speed,WPO_Red,WPO_Green]}
  n11 = VertexColor()
  n12 = ScalarParameter() {defaultValue=5, parameterName=Amount_Offset_Green}
  n13 = Multiply(A=@n12) {constB=-1}
  n14 = Add(
   A=Multiply(A=LinearInterpolate(A=@n9, B=@n8, Alpha=@n10.WPO_Red), B=@n11.R)
   B=Multiply(A=@n11.G, B=LinearInterpolate(A=@n13, B=@n12, Alpha=@n10.WPO_Green))
  )
  StaticSwitchParameter(
   True=Constant()
   False=Add(
    A=MaterialFunctionCall!(Speed=@n4.WPO_Speed, SmallWaveScle=@n5, WaveScale=@n6) {materialFunction=/Game/Effects/Fort_Effects/Materials/_Functions/Jiggle_WPO}.Result
    B=Multiply(A=@n7, B=@n14)
   )
  ) {parameterName=Use on Pile}
 ]
}
